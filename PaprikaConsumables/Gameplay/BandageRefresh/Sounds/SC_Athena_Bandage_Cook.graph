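# Sound cue for the bandage consumable's "cook" sound, played quietly and pitched up in the gadgets and items sound class.
# A random node picks between two equally weighted wave variations, then a modulator varies pitch and volume on each play.
# It borrows the default grenade cook attenuation, and a concurrency override allows one instance per owner, stopping the oldest when a new one starts.
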SoundCue /PaprikaConsumables/Gameplay/BandageRefresh/Sounds/SC_Athena_Bandage_Cook
root=Modulator_0
nodes [
 {id=Modulator_0, children [Random_0], properties {pitchMax=1.1, volumeMin=0.7}}
 {id=Random_0, children [WavePlayer_2,WavePlayer_3], properties {bShouldExcludeFromBranchCulling=true, weights [1,1]}}
 {id=WavePlayer_2, properties {soundWaveAssetPtr=&/Athena_Bandage_Cook_02}}
 {id=WavePlayer_3, properties {soundWaveAssetPtr=&/Athena_Bandage_Cook_01}}
]
properties {
 attenuationSettings=/Game/Sounds/Fort_Weapons/Grenades/Grenade_Default_Cook_Att
 bOverrideConcurrency=true
 maxDistance=3000
 pitchMultiplier=1.3
 soundClassObject=&Fort_Gadgets_Items
 volumeMultiplier=0.15
 concurrencyOverrides {bLimitToOwner=true, maxCount=1, resolutionRule=StopOldest}
}
